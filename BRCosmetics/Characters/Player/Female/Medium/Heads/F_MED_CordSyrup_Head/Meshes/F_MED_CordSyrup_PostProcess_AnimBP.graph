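# Post-process AnimBP for the CordSyrup female medium head that applies corrective offsets to chin and lower-lip facial bones.
# Each update copies the CTRL_expressions_jawOpen curve into JawOpen, which weights additive bone-space translations on the three UnderChin bones.
# Three chained LipLower nodes specify a small additive offset independent of the jaw curve.
# The update also tracks whether Emote_Celebration_Female_M is playing, but the AnimGraph does not read that flag.
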AnimBlueprint /BRCosmetics/Characters/Player/Female/Medium/Heads/F_MED_CordSyrup_Head/Meshes/F_MED_CordSyrup_PostProcess_AnimBP
target_skeleton=/Game/Characters/Player/Common/Fortnite_Base_Head/Fortnite_Base_3L_Head_Skeleton
graphs [
 AnimGraph {nodes [
   n1 = Get JawOpen() -> (JawOpen:double)
   n2 = "Transform (Modify) Bone - Bone: FACIAL_C_UnderChin"(ComponentPose=Local To Component(LocalPose=Input Pose()), Alpha=@n1, Translation="0,-2,-2", Rotation="0,0,0", Scale="1,1,1") {
    translationMode=BMM_Additive
    translationSpace=BCS_BoneSpace
    alphaBoolBlend {blendOption=Linear}
    boneToModify {boneName=FACIAL_C_UnderChin}
    translation {y=-2, z=-2}
   }
   n3 = "Transform (Modify) Bone - Bone: FACIAL_R_UnderChin"(ComponentPose=@n2, Alpha=@n1, Translation="0,-3,-2", Rotation="0,0,0", Scale="1,1,1") {
    translationMode=BMM_Additive
    translationSpace=BCS_BoneSpace
    alphaBoolBlend {blendOption=Linear}
    boneToModify {boneName=FACIAL_R_UnderChin}
    translation {y=-3, z=-2}
   }
   n4 = "Transform (Modify) Bone - Bone: FACIAL_L_UnderChin"(ComponentPose=@n3, Alpha=@n1, Translation="0,-3,-2", Rotation="0,0,0", Scale="1,1,1") {
    translationMode=BMM_Additive
    translationSpace=BCS_BoneSpace
    alphaBoolBlend {blendOption=Linear}
    boneToModify {boneName=FACIAL_L_UnderChin}
    translation {y=-3, z=-2}
   }
   n5 = "Transform (Modify) Bone - Bone: FACIAL_C_LipLower"!(ComponentPose=@n4, bAlphaBoolEnabled:bool=True, Translation="0,1.2,0", Rotation="0,0,0", Scale="1,1,1") -> (Pose) {
    alphaInputType=Bool
    translationMode=BMM_Additive
    translationSpace=BCS_BoneSpace
    alphaBoolBlend {blendOption=Linear}
    boneToModify {boneName=FACIAL_C_LipLower}
    translation {y=1.2}
   }
   n6 = "Transform (Modify) Bone - Bone: FACIAL_L_LipLower"!(ComponentPose=@n5, bAlphaBoolEnabled:bool=True, Translation="0,1.2,0", Rotation="0,0,0", Scale="1,1,1") -> (Pose) {
    alphaInputType=Bool
    translationMode=BMM_Additive
    translationSpace=BCS_BoneSpace
    alphaBoolBlend {blendOption=Linear}
    boneToModify {boneName=FACIAL_L_LipLower}
    translation {y=1.2}
   }
   n7 = "Transform (Modify) Bone - Bone: FACIAL_R_LipLower"!(ComponentPose=@n6, bAlphaBoolEnabled:bool=True, Translation="0,1.2,0", Rotation="0,0,0", Scale="1,1,1") -> (Pose) {
    alphaInputType=Bool
    translationMode=BMM_Additive
    translationSpace=BCS_BoneSpace
    alphaBoolBlend {blendOption=Linear}
    boneToModify {boneName=FACIAL_R_LipLower}
    translation {y=1.2}
   }
   Output Pose(Result=Component To Local(ComponentPose=@n7)) {layerGroup=None}
  ]}
 EventGraph {nodes [
   Event BlueprintUpdateAnimation()
   n1 = GetPartSkeletalMeshComponent()
   > Set JawOpen(JawOpen:double=GetCurveValue(self=GetAnimInstance(self=@n1), CurveName=CTRL_expressions_jawOpen)) -> (Output_Get:double)
   n2 = GetAnimInstance(self=GetOwnerSkeletalMeshComponent())
   > Is Valid(InputObject=@n2)
   > Set is Emote playing(is Emote playing:bool=Montage_IsPlaying(self=@n2, Montage=/BRCosmetics/Animation/Game/MainPlayer/Emotes/Emote_Celebration_Female_M)) -> (Output_Get:bool)
  ]}
]
variables [
 is Emote playing {type=boolean}
 JawOpen {type=number}
 Out Names {type="array<string>", default []}
]
